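# Translucent unlit material for the Creative Sequencer device's volume, drawing an animated swirl field with glowing box edges.
# Two warped, panning noise layers form swirls, shaped by fresnel, behind-surface distance, animated grain and a gradient.
# Edge masks from UV channels 1-3, scaled by width, height and depth, draw distance-adjusted frame lines.
# High and Epic quality add depth-fade intersection glow and near-camera fade before output passes through MF_CreativeMoveTool.
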Material /Game/Creative/Devices/Sequencer/Materials/M_Device_Sequencer_Volume_01
graph {
 properties {
  allowTranslucentCustomDepthWrites=true
  bEnableResponsiveAA=true
  bUseMaterialAttributes=true
  blendMode=BLEND_Translucent
  pixelDepthOffsetMode=PDOM_Legacy
  shadingModel=MSM_Unlit
  usedShadingModels=Unlit
 }
 nodes [
  n1 = Constant2Vector() {g=1.5, r=1.5}
  n2 = Divide(A=NamedRerouteUsage()) {constB=1000}
  n3 = Multiply(
   A=@n2
   B=ComponentMask(None=VectorParameter() {group=Highlight, parameterName=Highlight UV Tile, defaultValue=<rgba 0.2 0.2 0 0>}.RGB) {g=true, r=true}
  )
  n4 = VectorParameter() {group=Highlight, parameterName=Highlight PanSpeed, defaultValue=<rgba 0.015 -0.04 0 0>}
  n5 = Multiply(
   A=TextureSampleParameter2D(UVs=Panner(Coordinate=@n3, Speed=@n4.RGB) {speedX=0.05, speedY=-0.08}) {parameterName=Noise, texture=/Game/Global/Textures/Default/Noise/T_Dudebro_Noise_01}.R
   B=ScalarParameter() {defaultValue=0.2, group=Highlight, parameterName=Highlight Warp Intensity}
  ) {constB=0.2}
  Add(A=Panner(Coordinate=Multiply(A=@n2, B=@n1)) {speedX=-0.005, speedY=0.01}, B=@n5)
  n6 = Multiply(
   A=VectorParameter() {group=Depth Fade, parameterName=Depth Fade Edge Color, defaultValue=<rgba 1 0.1619 0.0321 1>}.RGB
   B=ScalarParameter() {defaultValue=2, group=Depth Fade, parameterName=Depth Fade Edge Intensity}
  )
  n7 = NamedRerouteUsage()
  n8 = ScalarParameter() {defaultValue=5, group=Depth Fade, parameterName=Depth Fade Edge Distance}
  n9 = NamedRerouteUsage()
  n10 = Divide(
   A=Panner(Coordinate=@n9, Speed=VectorParameter() {group=Swirls, parameterName=PanSpeedNoiseA, defaultValue=<rgba 7.5 15 11 0>}.RGB) {speedX=5}
   B=ScalarParameter() {defaultValue=1000, group=Swirls, parameterName=RepeatNoiseA}
  )
  n11 = ScalarParameter() {defaultValue=0.2, group=Swirls, parameterName=Warp Intensity}
  n12 = AppendVector(
   A=ScalarParameter() {defaultValue=2.4, group=Swirls, parameterName=Warp Tile U}
   B=ScalarParameter() {defaultValue=1.4, group=Swirls, parameterName=Warp Tile V}
  )
  n13 = TextureSampleParameter2D(UVs=Panner(
    Coordinate=Multiply(A=@n2, B=@n12)
    Speed=VectorParameter() {group=Swirls, parameterName=PanSpeedWarp, defaultValue=<rgba 0.015 -0.04 0 0>}.RGB
   ) {speedX=0.017, speedY=-0.03}) {group=Swirls, parameterName=Warp Texture, samplerType=SAMPLERTYPE_Grayscale, texture=&LowResBlurredNoise}
  n14 = Add(A=Multiply(A=@n11, B=@n13.R) {constB=0.2}, B=Multiply(A=@n11, B=Power(Base=@n13.R)) {constB=0.2})
  n15 = ChannelMaskParameter(None=TextureSampleParameter2D(UVs=Add(A=@n10, B=@n14)) {group=Swirls, parameterName=Noise A, texture=&T_Dudebro_Noise_01}.RGBA) {group=Swirls, parameterName=Noise A Channel}
  n16 = Divide(
   A=Panner(Coordinate=@n9, Speed=VectorParameter() {group=Swirls, parameterName=PanSpeedNoiseB, defaultValue=<rgba -11 -32.02 0 0>}.RGB) {speedX=5}
   B=ScalarParameter() {defaultValue=-1300, group=Swirls, parameterName=RepeatNoiseB}
  )
  n17 = Multiply(A=@n13.R, B=Multiply(A=@n11) {constB=0.8}) {constB=0.2}
  n18 = ChannelMaskParameter(None=TextureSampleParameter2D(UVs=Add(A=@n16, B=@n17)) {group=Swirls, parameterName=Noise B, texture=&T_Dudebro_Noise_01}.RGBA) {group=Swirls, maskChannel=Green, parameterName=Noise B Channel, defaultValue=<rgba 0 1 0 0>}
  n19 = Multiply(
   A=NamedRerouteDeclaration(None=Multiply(A=@n15.R, B=@n18.R)) {name=Swirls, nodeColor=<rgba 0 1 1 1>}
   B=ScalarParameter() {defaultValue=1, group=Swirls, parameterName=Swirls Intensity}
  )
  n20 = NamedRerouteUsage()
  n21 = Add(
   A=Add(A=@n19, B=Multiply(A=@n19, B=@n20))
   B=MaterialFunctionCall(
    Brightness=ScalarParameter() {defaultValue=0.5, parameterName=Fresnel Brightness}
    Exponent=ScalarParameter() {defaultValue=4, parameterName=Fresnel Exponent}
   ) {materialFunction=&MF_Fresnel_FX}.Fresnel Inverted
  )
  n22 = Multiply(
   A=MaterialFunctionCall!(Distance Bias, Distance Divisor=ScalarParameter() {defaultValue=20, parameterName="Distance Divisor (S)"}, Scene Depth, Pixel Depth) {materialFunction=/Game/Effects/Fort_Effects/Materials/_Functions/MF_DistanceBehindTranslucency_01}.Normalized Inverted
   B=ScalarParameter() {defaultValue=0.2, parameterName=DistanceBehindTranslucencyScale}
  )
  n23 = Saturate(None=Power(Base=Add(A=@n21, B=@n22), Exp=ScalarParameter() {defaultValue=2, group=Swirls, parameterName=Swirls Exponent}) {constExponent=5})
  n24 = NamedRerouteUsage()
  n25 = Multiply(A=@n23, B=Multiply(A=@n24, B=ScalarParameter() {defaultValue=1, group=Grain, parameterName=Grain Brighten Intensity}))
  n26 = VectorParameter() {parameterName=Base Color, defaultValue=<rgba 0.7448 0.0124 0 0>}
  n27 = Multiply(A=@n26.RGB, B=Multiply(A=@n23, B=@n24))
  n28 = Power(Base=OneMinus(None=@n23), Exp=ScalarParameter() {defaultValue=1, group=Grain, parameterName=Grain Darken Exp})
  n29 = Multiply(
   A=LinearInterpolate(
    A=LinearInterpolate(A=@n26.RGB, B=@n27, Alpha=@n28) {constB=0}
    B=VectorParameter() {parameterName=Swirls Color, defaultValue=<rgba 10 1.6189 0.3215 0>}.RGB
    Alpha=Add(A=@n23, B=@n25)
   )
   B=SphereMask(
    A=MaterialFunctionCall() {materialFunction=&LinearGradient}.VGradient
    B=Constant() {r=0.5}
    Radius=ScalarParameter() {defaultValue=0.5, parameterName=BGGradientRadius}
    Hardness=ScalarParameter() {parameterName=BGGradientHardness}
   )
  ) {desc=Add Gradient}
  n30 = Multiply(
   A=VectorParameter() {group=Frame, parameterName=Frame Glow Color, defaultValue=<rgba 1 0.1025 0 1>}.RGB
   B=ScalarParameter() {defaultValue=100, group=Frame, parameterName=Frame Glow Intensity}
  )
  n31 = TextureCoordinate() {coordinateIndex=3}
  n32 = ScalarParameter() {defaultValue=100, group=Frame, parameterName=Height, sortPriority=3}
  n33 = ScalarParameter() {defaultValue=2, desc=Frame Distance Multiplier, group=Frame Distance Fade, parameterName=Frame Thickness Multiplier}
  n34 = LinearInterpolate(A=@n32, B=Multiply(A=@n32, B=Divide(B=@n33)), Alpha=NamedRerouteUsage())
  n35 = Divide(B=ScalarParameter() {defaultValue=2, group=Frame, parameterName=Frame Thickness, sortPriority=60})
  n36 = Divide(A=Divide(B=ScalarParameter() {defaultValue=10, group=Frame, parameterName=Frame Falloff, sortPriority=59})) {constB=100}
  n37 = Saturate(None=MaterialFunctionCall!(UVs=@n31, Mult=Multiply(A=@n35, B=@n34), Exp=Multiply(A=@n36, B=@n34)) {materialFunction=/Game/Effects/Fort_Effects/MaterialFunctions/MF_SquareEdgeMask}.HorizMask)
  n38 = ComponentMask(None=@n31) {r=true}
  n39 = Constant() {r=1}
  n40 = Constant()
  n41 = TextureCoordinate() {coordinateIndex=2}
  n42 = ScalarParameter() {defaultValue=100, group=Frame, parameterName=Depth, sortPriority=2}
  n43 = LinearInterpolate(A=@n42, B=Multiply(A=@n42, B=Divide(B=@n33)), Alpha=NamedRerouteUsage())
  n44 = Multiply(A=@n35, B=@n43)
  n45 = Multiply(A=@n36, B=@n43)
  n46 = OneMinus(None=Saturate(None=MaterialFunctionCall(UVs=@n41, Mult=@n44, Exp=@n45) {materialFunction=&MF_SquareEdgeMask}.HorizMask))
  n47 = Constant() {r=1}
  n48 = If(A=ComponentMask(None=@n41) {r=true}, B=@n47, "A > B"=Constant(), "A == B"=@n47, "A < B"=@n47) {desc=Width}
  n49 = TextureCoordinate() {coordinateIndex=1}
  n50 = ScalarParameter() {defaultValue=100, group=Frame, parameterName=Width, sortPriority=1}
  n51 = LinearInterpolate(A=@n50, B=Multiply(A=@n50, B=Divide(B=@n33)), Alpha=NamedRerouteUsage())
  n52 = Multiply(A=@n35, B=@n51)
  n53 = Multiply(A=@n36, B=@n51)
  n54 = OneMinus(None=Saturate(None=MaterialFunctionCall(UVs=@n49, Mult=@n52, Exp=@n53) {materialFunction=&MF_SquareEdgeMask}.Vertmask))
  n55 = Constant() {r=1}
  n56 = If(A=ComponentMask(None=@n49) {r=true}, B=@n55, "A > B"=Constant(), "A == B"=@n55, "A < B"=@n55) {desc=Width}
  n57 = Max(
   A=Multiply(A=OneMinus(None=@n37), B=If(A=@n38, B=@n39, "A > B"=@n40, "A == B"=@n39, "A < B"=@n39) {desc=Height})
   B=Max(A=Multiply(A=@n46, B=@n48), B=Multiply(A=@n54, B=@n56))
  )
  n58 = LinearInterpolate(
   A=@n29
   B=LinearInterpolate(
    A=LinearInterpolate(A=@n29, B=@n30, Alpha=@n57) {desc=Add Frame Glow}
    B=VectorParameter() {group=Frame, parameterName=Frame Color, defaultValue=<rgba 1 1 1 1>}.RGB
    Alpha=@n57
   ) {desc=Add frame}
   Alpha=SphereMask(
    A=PixelDepth().R
    B=Constant()
    Radius=ScalarParameter() {defaultValue=2500, group=Frame Distance Fade, parameterName=Frame Far Distance Fade, sortPriority=60}
   ) {hardnessPercent=10}
  ) {desc=Add Frame Lines}
  n59 = LinearInterpolate(
   A=Power(Base=Multiply(A=@n6, B=@n7), Exp=ScalarParameter() {group=Depth Fade, parameterName=Grain Exp}) {constExponent=1.5}
   B=@n58
   Alpha=Power(Base=DepthFade(FadeDistance=@n8), Exp=ScalarParameter() {defaultValue=2, group=Depth Fade, parameterName=Depth Fade Edge Exp})
  )
  n60 = Divide(A=@n8) {constB=5}
  n61 = MaterialFunctionCall(Distance=ScalarParameter() {defaultValue=200, parameterName=Near Camera Fade Distance}) {materialFunction=&Near_Camera_Fade}
  n62 = Multiply(A=LinearInterpolate(A=@n6, B=Add(A=@n58, B=@n59), Alpha=DepthFade(FadeDistance=@n60)), B=@n61.Result)
  n63 = Constant() {r=1}
  n64 = DepthFade(Opacity=@n61.Result, FadeDistance=ScalarParameter() {defaultValue=5, group=Depth Fade, parameterName=Depth Fade Distance})
  n65 = QualitySwitch(Default=@n63, Low=@n63, High=@n64, Medium=@n63, Epic=@n64)
  n66 = ScalarParameter() {defaultValue=1, parameterName=Opacity}
  MaterialFunctionCall(MA In=SetMaterialAttributes!(
    MaterialAttributes
    Emissive Color=QualitySwitch(Default=@n58, Low=@n58, High=@n62, Medium=@n58, Epic=@n62)
    Opacity=Multiply(A=@n65, B=@n66)
   ) {attributeSetTypes [#1, #2]}) {materialFunction=&MF_CreativeMoveTool}
  Multiply(A=@n35) {constB=10}
  n67 = MaterialFunctionCall(Vector 2=ObjectBounds()) {materialFunction=/Engine/Functions/Engine_MaterialFunctions02/Utility/VectorLength}
  n68 = Distance(A=ActorPositionWS(), B=CameraPositionWS())
  Multiply(A=MaterialFunctionCall(UVs=NamedRerouteUsage(), Texture Scale=Divide(A=@n67.V2 Length, B=@n68)) {materialFunction=&ScaleUVsByCenter}.UVs)
  n69 = Constant()
  n70 = Distance(A=CameraPositionWS(), B=WorldPosition().XYZ)
  n71 = ScalarParameter() {defaultValue=-100, group=Frame Distance Fade, parameterName=Frame Thickness DistanceBias}
  n72 = Saturate(None=Divide(A=Add(A=@n70, B=@n71), B=ScalarParameter() {defaultValue=1200, group=Frame Distance Fade, parameterName=Frame Thickness FadeScale}))
  NamedRerouteDeclaration(None=QualitySwitch(Default=@n69, Low=@n69, High=@n72, Medium=@n69, Epic=@n72)) {name=Frame Distance Fade for Scaling, nodeColor=<rgba 1 0.6328 0 1>}
  n73 = MaterialFunctionCall(Float3=Multiply(A=WorldPosition().XYZ) {constB=-1}) {materialFunction=&BreakOutFloat3Components}
  n74 = AppendVector(A=@n73.G, B=@n73.G)
  n75 = AppendVector(A=@n73.R, B=@n73.R)
  n76 = VertexNormalWS()
  n77 = MaterialFunctionCall(Float3=Abs(None=Round(None=@n76))) {materialFunction=&BreakOutFloat3Components}
  NamedRerouteDeclaration(None=LinearInterpolate(A=LinearInterpolate(A=@n74, B=@n75, Alpha=@n77.G), B=AppendVector(A=@n73.R, B=@n73.R), Alpha=@n77.B)) {name=WorldAlignedUVs, nodeColor=<rgba 0.9497 0 1 1>}
  n78 = NamedRerouteUsage()
  n79 = ScalarParameter() {defaultValue=100, group=Grain, parameterName=Grain Scale}
  n80 = Divide(A=@n78, B=Multiply(A=@n79) {constB=1.7}) {constB=100}
  n81 = VectorParameter() {group=Grain, parameterName=Grain PanSpeed, defaultValue=<rgba 0.03 -0.1 0 1>}
  n82 = Multiply(A=@n81.RGB) {constB=-1.3}
  n83 = Panner(Coordinate=Divide(A=@n78, B=@n79) {constB=100}, Speed=@n81.RGB) {speedX=0.03, speedY=-0.1}
  n84 = Multiply(A=Time(), B=ScalarParameter() {defaultValue=0.7, group=Grain, parameterName=Grain Sine Speed})
  n85 = ScalarParameter() {defaultValue=0.2, group=Grain, parameterName=Grain Sine Range}
  n86 = Add(A=Multiply(A=Sine(None=@n84) {period=3}, B=@n85) {constB=0.2}, B=Divide(A=@n85))
  n87 = ScalarParameter() {defaultValue=0.5, group=Grain, parameterName=Grain Sine Base}
  n88 = Multiply(
   A=TextureSampleParameter2D(UVs=Panner(Coordinate=@n80, Speed=@n82) {speedX=0.03, speedY=-0.1}) {parameterName=Noise Grain, texture=/Game/Global/Textures/Default/Noise/NoiseDigital_large}.G
   B=Multiply(
    A=TextureSampleParameter2D(UVs=@n83) {parameterName=Noise Grain, texture=&NoiseDigital_large}.R
    B=Add(A=@n86, B=@n87)
   )
  ) {desc=Combine Animated Grain Layers}
  n89 = NamedRerouteUsage()
  NamedRerouteDeclaration(None=Power(Base=Multiply(A=@n88, B=@n89), Exp=ScalarParameter() {defaultValue=1, group=Grain, parameterName=Grain Power})) {name=Noise Grain, nodeColor=<rgba 0.7109 0 1 1>}
  NamedRerouteDeclaration(None=@n88) {name=Depth Edge Grain, nodeColor=<rgba 1 0.98 0 1>}
  n90 = NamedRerouteUsage()
  n91 = Multiply(A=Time(), B=ScalarParameter() {defaultValue=0.35, group=Highlight, parameterName=Highlight SineSpeed})
  n92 = ConstantBiasScale(None=Sine(None=Add(A=@n91, B=@n5)) {period=3}) {bias=0, scale=1}
  n93 = ScalarParameter() {defaultValue=0.5, group=Highlight, parameterName=Highlight Radius}
  n94 = ScalarParameter() {group=Highlight, parameterName=Highlight Hardness}
  n95 = Multiply(
   A=Power(
    Base=SphereMask(A=@n90, B=@n92, Radius=@n93, Hardness=@n94)
    Exp=ScalarParameter() {defaultValue=1, group=Highlight, parameterName=Highlight Power}
   )
   B=ScalarParameter() {defaultValue=1, group=Highlight, parameterName=Highlight Intensity}
  )
  NamedRerouteDeclaration(None=@n95) {name=Highlight, nodeColor=<rgba 1 0.9141 0 1>}
  ParticleColor()
  n96 = Power(Base=@n57, Exp=ScalarParameter() {defaultValue=0.6, group=Frame Distance Fade, parameterName=Exp})
  n97 = ScalarParameter() {defaultValue=2, group=Frame Distance Fade, parameterName=Multiplier}
  Reroute(None=LinearInterpolate(A=@n57, B=Multiply(A=@n96, B=@n97) {constB=2}, Alpha=NamedRerouteUsage()))
  Reroute(None=@n88)
  Reroute(None=@n95)
  Reroute()
 ]
}
